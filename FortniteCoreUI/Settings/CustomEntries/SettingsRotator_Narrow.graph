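# Narrow settings rotator widget, derived from FortSettingsRotator, that shows the current option as centered uppercase text over a material-driven Border background.
# Options-populated, option-selected and default-specified events push the count and indices into the background's dynamic material as scalar parameters, leaving any option indicator to the material.
# Hover and unhover custom events play RotatorHover forward or in reverse, brightening the text and growing a spacer beneath it.
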WidgetBlueprint /FortniteCoreUI/Settings/CustomEntries/SettingsRotator_Narrow
animations [RotatorHover {
  binding_count=2
  length_seconds=0.1
  object=&self:RotatorHover
  sequence {bindings [
    MyText {
     id=#1
     possessed_class=CommonTextBlock
     tracks [{
       type=MovieSceneFloatTrack
       display_name=Render Opacity
       sections [{type=MovieSceneFloatSection, end_frame=2, start_frame=0, channels [None {type=MovieSceneScriptingFloatChannel, keys [[0,0.5],[2,1]]}]}]
      }]
    }
    Spacer_135 {
     id=#2
     possessed_class=Spacer
     tracks [{
       type=MovieSceneFloatVectorTrack
       display_name=Size
       sections [{
         type=MovieSceneFloatVectorSection
         channels [
          Vector.X {type=MovieSceneScriptingFloatChannel, const [1,1]}
          Vector.Y {type=MovieSceneScriptingFloatChannel, keys [[0,1],[2,5]]}
         ]
        }]
      }]
    }
   ]}
 }]
graphs [EventGraph {nodes [
   OnHover Animation() {is=title K2Node_CustomEvent}
   > PlayAnimation(InAnimation=Get RotatorHover() -> (RotatorHover:Widget Animation Object Reference))
   OnUnHover Animation() {is=title K2Node_CustomEvent}
   > PlayAnimation(InAnimation=Get RotatorHover(), PlayMode=Reverse)
   n1 = Event On Options Populated!() -> (OutputDelegate, then, Count:int=0) {is=title K2Node_Event}
   > n2 = GetDynamicMaterial(self:Border Object Reference=Get Background() -> (Background:Border Object Reference))
   > SetScalarParameterValue(self=@n2, ParameterName=Options, Value="To Float (Integer)"(InInt=@n1.Count))
   n3 = Event On Options Selected!() -> (OutputDelegate, then, Index:int=0) {is=title K2Node_Event}
   > n4 = GetDynamicMaterial(self=Get Background())
   > SetScalarParameterValue(self=@n4, ParameterName=SelectedOption, Value="To Float (Integer)"(InInt=@n3.Index))
   n5 = Event On Default Option Specified!() -> (OutputDelegate, then, DefaultOptionIndex:int=0) {is=title K2Node_Event}
   > n6 = GetDynamicMaterial(self=Get Background())
   > SetScalarParameterValue(self=@n6, ParameterName=DefaultOption, Value="To Float (Integer)"(InInt=@n5.DefaultOptionIndex))
  ]}]
variables []
widgets {
 info {inheritedWidgetCount=2, parentClass=/Script/FortniteUI.FortSettingsRotator, rootWidgetClass=&Border, widgetCount=5}
 widgets [
  Background {
   widgetClassPath=&Border
   bIsVariable=true
   widget=Background
   properties {
    verticalAlignment=VAlign_Center
    background {
     bIsDynamicallyLoaded=false
     drawAs=Image
     imageType=NoImage
     mirroring=NoMirror
     resourceName=None
     resourceObject=/FortniteCoreUI/Settings/Art/SettingsRotator_M
     tiling=NoTile
     imageSize=<32 32>
     margin=<ltrb 0 0 0 0>
     outlineSettings {
      bUseBrushTransparency=false
      roundingType=HalfHeightRadius
      width=0
      color {colorUseRule=UseColor_Specified, specifiedColor=<rgba 0 0 0 0>}
      cornerRadii=<0 0 0 0>
     }
     tintColor {colorUseRule=UseColor_Specified, specifiedColor=<rgba 1 1 1 1>}
     uVRegion {bIsValid=false, max=<0 0>, min=<0 0>}
    }
    padding=<ltrb 0 2 0 6>
   }
  }
  VerticalBox_0 {
   widgetClassPath=&VerticalBox
   parent=Background
   widget=VerticalBox_0
   slot {class=BorderSlot, properties {verticalAlignment=VAlign_Center, padding=<ltrb 0 2 0 6>}}
  }
  MyText {
   widgetClassPath=&CommonTextBlock
   bInherited=true
   bIsVariable=true
   parent=VerticalBox_0
   widget=MyText
   properties {
    justification=Center
    mobileFontSizeMultiplier=1.15
    scrollStyle=&TextScrollStyle-Base_C
    text="{Option}"
    textTransformPolicy=ToUpper
    colorAndOpacity {colorUseRule=UseColor_Specified, specifiedColor=<rgba 0.6105 0.8796 1 1>}
    font {
     bForceMonospaced=false
     bMaterialIsStencil=false
     fontMaterial=null
     fontObject=&BurbankSmall
     letterSpacing=0
     monospacedWidth=1
     size=14
     skewAmount=0
     typefaceFontName=Black
     outlineSettings {
      bApplyOutlineToDropShadows=false
      bMiteredCorners=false
      bSeparateFillAlpha=false
      outlineMaterial=null
      outlineSize=0
      outlineColor=<rgba 0 0 0 1>
     }
    }
    shadowOffset=<0 0>
   }
   slot {class=VerticalBoxSlot}
  }
  Spacer_135 {
   widgetClassPath=&Spacer
   parent=VerticalBox_0
   widget=Spacer_135
   slot {class=VerticalBoxSlot, properties {verticalAlignment=VAlign_Center}}
  }
  InputActionWidget {widgetClassPath=&CommonActionWidget, bInherited=true, widget=null}
 ]
}
